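# Sound cue named as growth ambience for the DudeBro CubeIsland event in Figment Season 10 Athena content.
# The cue has no root node and an empty node graph, so it references no audio and produces no sound as stored.
# It is assigned to the extra-high priority events sound class.
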SoundCue /Figment_S10/Athena/Sounds/Events/CubeIsland/DudeBro_Amb_Growth_Cue
root=null
nodes []
properties {maxDistance=2097152, soundClassObject=/Figment_S10/Athena/Sounds/Events/SoundClasses/Fort_Events_Priority_ExtHigh, volumeMultiplier=1}
